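# Physics asset for the Rogue Ninja male medium body's face accessory mesh, simulating its front hair strands.
# Small capsule bodies on the dyn_hairFront bones form single links and chains hanging off a head capsule, which joins a neck capsule.
# Hair constraints lock twist and carry tight swing limit angles, while the head-to-neck constraint leaves swing and twist free.
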PhysicsAsset /BRCosmetics/Characters/Player/Male/Medium/Bodies/M_MED_RogueNinja/Meshes/Parts/M_MED_RogueNinja_FaceAcc_Physics
bodies {
 dyn_hairFrontA [{shapeType=Capsule, length=0.5, radius=0.5, shapeName=dyn_hairFrontA_capsule}]
 dyn_hairFrontB [{shapeType=Capsule, length=0.5, radius=0.5, shapeName=dyn_hairFrontB_capsule}]
 dyn_hairFrontC1 [{shapeType=Capsule, length=0.5, radius=0.5, shapeName=dyn_hairFrontC1_capsule, rotation=<rot 90 0 0>}]
 dyn_hairFrontC2 [{shapeType=Capsule, length=0.5, radius=0.5, shapeName=dyn_hairFrontC2_capsule, rotation=<rot 90 0 0>}]
 dyn_hairFrontC3 [{shapeType=Capsule, length=0.5, radius=0.5, shapeName=dyn_hairFrontC3_capsule, rotation=<rot 90 0 0>}]
 dyn_hairFrontD [{shapeType=Capsule, length=1, radius=0.5, shapeName=dyn_hairFrontD_capsule, rotation=<rot 90 0 0>}]
 dyn_hairFrontD1 [{shapeType=Capsule, length=1, radius=0.5, shapeName=dyn_hairFrontD1_capsule, rotation=<rot 90 0 0>}]
 dyn_hairFrontD2 [{shapeType=Capsule, length=1, radius=0.5, shapeName=dyn_hairFrontD2_capsule, rotation=<rot 90 0 0>}]
 dyn_hairFrontD3 [{shapeType=Capsule, length=1, radius=0.5, shapeName=dyn_hairFrontD3_capsule, rotation=<rot 90 0 0>}]
 dyn_hairFrontD4 [{shapeType=Capsule, length=1, radius=0.5, shapeName=dyn_hairFrontD4_capsule, rotation=<rot 90 0 0>}]
 dyn_hairFrontE1 [{shapeType=Capsule, length=1, radius=0.5, shapeName=dyn_hairFrontE1_capsule, rotation=<rot 90 0 0>}]
 dyn_hairFrontE2 [{shapeType=Capsule, length=1, radius=0.5, shapeName=dyn_hairFrontE2_capsule, rotation=<rot 90 0 0>}]
 dyn_hairFrontF [{shapeType=Capsule, length=0.8554, radius=0.539, shapeName=dyn_hairFrontF_capsule, rotation=<rot 90 0 0>}]
 dyn_hairFrontG1 [{shapeType=Capsule, length=0.5, radius=0.5, shapeName=dyn_hairFrontG1_capsule, rotation=<rot 90 0 0>}]
 dyn_hairFrontG2 [{shapeType=Capsule, length=0.5, radius=0.5, shapeName=dyn_hairFrontG2_capsule, rotation=<rot 90 0 0>}]
 dyn_hairFrontG3 [{shapeType=Capsule, length=0.5, radius=0.5, shapeName=dyn_hairFrontG3_capsule, rotation=<rot 90 0 0>}]
 dyn_hairFrontH1 [{shapeType=Capsule, length=0.5, radius=0.5, shapeName=dyn_hairFrontH1_capsule, rotation=<rot 90 0 0>}]
 dyn_hairFrontH2 [{shapeType=Capsule, length=0.5, radius=0.5, shapeName=dyn_hairFrontH2_capsule, rotation=<rot 90 0 0>}]
 head [{shapeType=Capsule, length=5, radius=8, center=<-0.1375 0.7885 -0.0397>, rotation=<rot 90 178.2101 168.319>}]
 neck_01 [{
   shapeType=Capsule
   length=21.9333
   radius=2.0667
   shapeName=neck_01_capsule
   center=<11.7741 3.239 9.6055>
   rotation=<rot 84.5979 147.9205 -171.8738>
  }]
}
constraints [
 {bone1Name=dyn_hairFrontD, bone2Name=head, swing1LimitDegrees=7, swing2LimitDegrees=7, twistLimitDegrees=45, twistMotion=Locked}
 {bone1Name=dyn_hairFrontD1, bone2Name=dyn_hairFrontD, swing1LimitDegrees=7, swing2LimitDegrees=7, twistLimitDegrees=45, twistMotion=Locked}
 {bone1Name=dyn_hairFrontD2, bone2Name=dyn_hairFrontD1, swing1LimitDegrees=7, swing2LimitDegrees=7, twistLimitDegrees=45, twistMotion=Locked}
 {bone1Name=dyn_hairFrontD3, bone2Name=dyn_hairFrontD2, swing1LimitDegrees=7, swing2LimitDegrees=7, twistLimitDegrees=45, twistMotion=Locked}
 {bone1Name=dyn_hairFrontD4, bone2Name=dyn_hairFrontD3, swing1LimitDegrees=7, swing2LimitDegrees=7, twistLimitDegrees=45, twistMotion=Locked}
 {bone1Name=dyn_hairFrontE1, bone2Name=head, swing1LimitDegrees=10, swing2LimitDegrees=10, twistLimitDegrees=45, twistMotion=Locked}
 {bone1Name=dyn_hairFrontE2, bone2Name=dyn_hairFrontE1, swing1LimitDegrees=10, swing2LimitDegrees=10, twistLimitDegrees=45, twistMotion=Locked}
 {bone1Name=dyn_hairFrontF, bone2Name=head, swing1LimitDegrees=10, swing2LimitDegrees=10, twistLimitDegrees=45, twistMotion=Locked}
 {bone1Name=dyn_hairFrontG1, bone2Name=head, swing1LimitDegrees=10, swing2LimitDegrees=10, twistLimitDegrees=45, twistMotion=Locked}
 {bone1Name=dyn_hairFrontG2, bone2Name=dyn_hairFrontG1, swing1LimitDegrees=10, swing2LimitDegrees=10, twistLimitDegrees=45, twistMotion=Locked}
 {bone1Name=dyn_hairFrontG3, bone2Name=dyn_hairFrontG2, swing1LimitDegrees=10, swing2LimitDegrees=10, twistLimitDegrees=45, twistMotion=Locked}
 {bone1Name=dyn_hairFrontH1, bone2Name=head, swing1LimitDegrees=10, swing2LimitDegrees=10, twistLimitDegrees=45, twistMotion=Locked}
 {bone1Name=dyn_hairFrontH2, bone2Name=dyn_hairFrontH1, swing1LimitDegrees=10, swing2LimitDegrees=10, twistLimitDegrees=45, twistMotion=Locked}
 {bone1Name=dyn_hairFrontA, bone2Name=head, swing1LimitDegrees=10, swing2LimitDegrees=10, twistLimitDegrees=45, twistMotion=Locked}
 {bone1Name=dyn_hairFrontB, bone2Name=head, swing1LimitDegrees=10, swing2LimitDegrees=10, twistLimitDegrees=45, twistMotion=Locked}
 {bone1Name=dyn_hairFrontC3, bone2Name=dyn_hairFrontC2, swing1LimitDegrees=9, swing2LimitDegrees=13, twistLimitDegrees=45, twistMotion=Locked}
 {bone1Name=dyn_hairFrontC2, bone2Name=dyn_hairFrontC1, swing1LimitDegrees=8, swing2LimitDegrees=13, twistLimitDegrees=45, twistMotion=Locked}
 {bone1Name=dyn_hairFrontC1, bone2Name=head, swing1LimitDegrees=10, swing2LimitDegrees=13, twistLimitDegrees=45, twistMotion=Locked}
 {
  bone1Name=head
  bone2Name=neck_01
  swing1LimitDegrees=45
  swing1Motion=Free
  swing2LimitDegrees=45
  swing2Motion=Free
  twistLimitDegrees=45
  twistMotion=Free
 }
]
